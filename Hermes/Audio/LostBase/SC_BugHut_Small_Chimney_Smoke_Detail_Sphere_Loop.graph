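# Ambient SoundCue for a looping chimney smoke detail spot sound at a small BugHut in LostBase.
# A random pick of three looping waves gets pitch variation, then passes through endlessly looping pitch and volume envelopes with different periods.
# An attenuation node adds logarithmic falloff, low-pass and listener focus.
# The root randomly starts that shared chain directly or after a short fixed delay, staggering instances, and the cue sends to the exterior ambient bus.
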SoundCue /Hermes/Audio/LostBase/SC_BugHut_Small_Chimney_Smoke_Detail_Sphere_Loop
root=Random_0
nodes [
 {
  id=Attenuation_1
  children [Enveloper_3]
  properties {
   bOverrideAttenuation=true
   attenuationOverrides {
    bAttenuateWithLPF=true
    bEnableListenerFocus=true
    distanceAlgorithm=Logarithmic
    falloffDistance=1000
    focusVolumeAttenuation=0.09
    hPFFrequencyAtMax=200
    hPFFrequencyAtMin=200
    lPFFrequencyAtMax=1000
    lPFRadiusMax=3000
    lPFRadiusMin=1000
    nonFocusVolumeAttenuation=0.01
    nonSpatializedRadiusEnd=750
    nonSpatializedRadiusStart=1000
    stereoSpread=1000
    attenuationShapeExtents=<250 500>
   }
  }
 }
 {id=Delay_0, children [Attenuation_1], properties {delayMax=0.35, delayMin=0.35}}
 {
  id=Enveloper_0
  children [Modulator_0]
  properties {
   bLoop=true
   bLoopIndefinitely=true
   loopEnd=30
   pitchCurve {editorCurveData {keys [
      {arriveTangent=-0.0013, interpMode=RCIM_Cubic, leaveTangent=-0.0013, time=0}
      <at 15 0.98 RCIM_Cubic>
      {arriveTangent=0.0013, interpMode=RCIM_Cubic, leaveTangent=0.0013, time=30}
     ]}}
   volumeCurve {editorCurveData {keys [<at 0 1>]}}
  }
 }
 {
  id=Enveloper_3
  children [Enveloper_0]
  properties {
   bLoop=true
   bLoopIndefinitely=true
   loopEnd=20
   pitchCurve {editorCurveData {keys [{time=0}]}}
   volumeCurve {editorCurveData {keys [
      {arriveTangent=-0.01, interpMode=RCIM_Cubic, leaveTangent=-0.01, time=0, value=1}
      <at 10 0.9 RCIM_Cubic>
      {arriveTangent=0.01, interpMode=RCIM_Cubic, leaveTangent=0.01, time=20, value=1}
     ]}}
  }
 }
 {id=Modulator_0, children [Random_1], properties {pitchMax=1.1, pitchMin=0.9}}
 {id=Random_0, children [Delay_0, Attenuation_1], properties {weights [1,1]}}
 {id=Random_1, children [WavePlayer_1,WavePlayer_2,WavePlayer_3], properties {weights [1,1,1]}}
 {
  id=WavePlayer_1
  properties {bLooping=true, soundWaveAssetPtr=&/Spot_BugHut_Small_Chimney_Smoke_Release_Detail_Sphere_Loop_01}
 }
 {
  id=WavePlayer_2
  properties {bLooping=true, soundWaveAssetPtr=&/Spot_BugHut_Small_Chimney_Smoke_Release_Detail_Sphere_Loop_02}
 }
 {
  id=WavePlayer_3
  properties {bLooping=true, soundWaveAssetPtr=&/Spot_BugHut_Small_Chimney_Smoke_Release_Detail_Sphere_Loop_03}
 }
]
properties {
 maxDistance=1250
 pitchMultiplier=1.2
 soundClassObject=&Fort_Ambient_Outdoors_Athena
 subtitlePriority=5000
 volumeMultiplier=0.25
 attenuationOverrides {
  attenuationShape=Capsule
  bAttenuate=false
  bAttenuateWithLPF=true
  bEnableFocusInterpolation=true
  dBAttenuationAtMax=-40
  falloffDistance=5000
  lPFFrequencyAtMax=1000
  lPFRadiusMax=3000
  lPFRadiusMin=1000
  nonFocusVolumeAttenuation=0.4
  occlusionInterpolationTime=1
  occlusionLowPassFilterFrequency=10000
  occlusionTraceChannel=ECC_GameTraceChannel11
  occlusionVolumeAttenuation=0.5
  stereoSpread=512
  attenuationShapeExtents=<2031.3811 553.7532 90>
 }
 busSends [{
   audioBus=null
   maxSendDistance=5000
   minSendDistance=1000
   minSendLevel=0
   sendLevel=1
   soundSourceBus=&AmbientAudio_Exterior_SourceBus
   customSendLevelCurve {editorCurveData {keys [<at 0 0.4>,<at 1 1>]}}
  }]
 concurrencyOverrides {maxCount=1, voiceStealReleaseTime=2}
 concurrencySet [/AmbientAudioBase/Concurrency/AmbientConc_2Max_Spot_Concurrency]
}
