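# Ambient outdoor looping SoundCue for small BugHut chimney smoke, reusing the large chimney smoke release loop.
# The looping wave feeds two capsule attenuation layers, one with listener focus and one with low-pass filtering, which are mixed and randomly pitched down.
# Looping envelopers on different periods slowly drift pitch and volume to disguise repetition.
# The root randomly adds a short start delay, and the cue sends to the exterior ambience bus.
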SoundCue /Hermes/Audio/LostBase/SC_BugHut_Small_Chimney_Smoke_Main_Capsule_Loop
root=Random_0
nodes [
 {
  id=Attenuation_0
  children [Mixer_0]
  properties {
   bOverrideAttenuation=true
   attenuationOverrides {
    attenuationShape=Capsule
    bEnableFocusInterpolation=true
    bEnableListenerFocus=true
    distanceAlgorithm=Logarithmic
    falloffDistance=950
    focusAttackInterpSpeed=0.09
    focusReleaseInterpSpeed=0.01
    lPFFrequencyAtMax=1000
    lPFRadiusMax=3000
    lPFRadiusMin=1000
    nonFocusVolumeAttenuation=0.4
    nonSpatializedRadiusEnd=200
    nonSpatializedRadiusStart=350
    stereoSpread=512
    attenuationShapeExtents=<550 200>
   }
  }
 }
 {
  id=Attenuation_1
  children [Mixer_0]
  properties {
   bOverrideAttenuation=true
   attenuationOverrides {
    attenuationShape=Capsule
    bAttenuateWithLPF=true
    bEnableFocusInterpolation=true
    distanceAlgorithm=Logarithmic
    falloffDistance=950
    hPFFrequencyAtMax=200
    hPFFrequencyAtMin=200
    lPFFrequencyAtMax=1000
    lPFRadiusMax=1000
    lPFRadiusMin=1000
    nonFocusVolumeAttenuation=0.4
    nonSpatializedRadiusEnd=250
    nonSpatializedRadiusStart=550
    stereoSpread=1000
    attenuationShapeExtents=<550 200>
   }
  }
 }
 {id=Delay_0, children [Enveloper_3], properties {delayMax=0.35, delayMin=0.35}}
 {
  id=Enveloper_0
  children [Modulator_0]
  properties {
   bLoop=true
   bLoopIndefinitely=true
   loopEnd=30
   pitchCurve {editorCurveData {keys [
      {arriveTangent=-0.0013, interpMode=RCIM_Cubic, leaveTangent=-0.0013, time=0}
      <at 15 0.98 RCIM_Cubic>
      {arriveTangent=0.0013, interpMode=RCIM_Cubic, leaveTangent=0.0013, time=30}
     ]}}
   volumeCurve {editorCurveData {keys [<at 0 1>]}}
  }
 }
 {
  id=Enveloper_3
  children [Enveloper_0]
  properties {
   bLoop=true
   bLoopIndefinitely=true
   loopEnd=20
   pitchCurve {editorCurveData {keys [{time=0}]}}
   volumeCurve {editorCurveData {keys [
      {arriveTangent=-0.01, interpMode=RCIM_Cubic, leaveTangent=-0.01, time=0, value=1}
      <at 10 0.9 RCIM_Cubic>
      {arriveTangent=0.01, interpMode=RCIM_Cubic, leaveTangent=0.01, time=20, value=1}
     ]}}
  }
 }
 {id=Mixer_0, children [WavePlayer_1], properties {inputVolume [1]}}
 {id=Mixer_2, children [Attenuation_1, Attenuation_0], properties {inputVolume [0.75,0.75]}}
 {id=Modulator_0, children [Mixer_2], properties {pitchMax=1, pitchMin=0.8}}
 {id=Random_0, children [Delay_0,Enveloper_3], properties {weights [1,1]}}
 {id=WavePlayer_1, properties {bLooping=true, soundWaveAssetPtr=&/Spot_BugHut_Large_Chimney_Smoke_Release_Loop}}
]
properties {
 maxDistance=1500
 soundClassObject=&Fort_Ambient_Outdoors_Athena
 subtitlePriority=5000
 volumeMultiplier=0.3
 attenuationOverrides {
  attenuationShape=Capsule
  bAttenuate=false
  bAttenuateWithLPF=true
  bEnableFocusInterpolation=true
  dBAttenuationAtMax=-40
  falloffDistance=5000
  lPFFrequencyAtMax=1000
  lPFRadiusMax=3000
  lPFRadiusMin=1000
  nonFocusVolumeAttenuation=0.4
  occlusionInterpolationTime=1
  occlusionLowPassFilterFrequency=10000
  occlusionTraceChannel=ECC_GameTraceChannel11
  occlusionVolumeAttenuation=0.5
  stereoSpread=512
  attenuationShapeExtents=<2031.3811 553.7532 90>
 }
 busSends [{
   audioBus=null
   maxSendDistance=5000
   minSendDistance=1000
   minSendLevel=0
   sendLevel=1
   soundSourceBus=&AmbientAudio_Exterior_SourceBus
   customSendLevelCurve {editorCurveData {keys [<at 0 0.4>,<at 1 1>]}}
  }]
 concurrencyOverrides {maxCount=1, voiceStealReleaseTime=2}
 concurrencySet [/AmbientAudioBase/Concurrency/AmbientConc_3Max_Spot_Default_Concurrency]
}
